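# Physics asset for the right-hand mesh of the Firework Female one-handed pickaxe cosmetic, providing its simple collision shape.
# It holds a single capsule body on the Melee_TwoHanded_Handle bone, offset up the handle.
# With only one body, it has no constraints.
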PhysicsAsset /BRCosmetics/Weapons/FORT_Melee/Pickaxe_FireworkFemale1h/Meshes/Firework_Axe_R_PhysicsAsset
bodies {Melee_TwoHanded_Handle [{shapeType=Capsule, length=39.8193, radius=12.995, center=<-0.2761 0 17.6336>, rotation=<rot -0.0004 -90 1.4544>}]}
constraints []
